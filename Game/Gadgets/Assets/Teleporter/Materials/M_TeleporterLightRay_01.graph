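# Translucent, unlit, two-sided material for animated teleporter light rays, blending between dark and bright blue color parameters.
# Two channels of a light-ray texture pan in opposite directions and multiply, then combine with vertex red and alpha into a clamped mask driving the color blend.
# That mask, scaled by vertex alpha, a Fresnel function and an opacity parameter, feeds a second blend that uses a high power of vertex alpha to push the strongest areas toward solid.
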Material /Game/Gadgets/Assets/Teleporter/Materials/M_TeleporterLightRay_01
graph {
 properties {blendMode=BLEND_Translucent, pixelDepthOffsetMode=PDOM_Legacy, shadingModel=MSM_Unlit, twoSided=true, usedShadingModels=Unlit}
 nodes [
  n1 = Panner() {bFractionalPart=true, speedX=0.132}
  n2 = Panner() {bFractionalPart=true, speedX=-0.1}
  n3 = VertexColor()
  n4 = Add(
   A=Multiply(
    A=TextureSample(UVs=@n1) {texture=/Game/Effects/Fort_Effects/Textures/Outlander/T_InTheZone_LightRays_01}.R
    B=TextureSample(UVs=@n2) {texture=&T_InTheZone_LightRays_01}.G
   )
   B=@n3.A
  )
  n5 = Power(Base=@n3.A) {constExponent=14}
  n6 = Clamp(None=Add(A=Add(A=@n4, B=@n3.R), B=@n5))
  LinearInterpolate(
   A=VectorParameter() {parameterName=Ray Color Dark, defaultValue=<rgba 0.0125 0.1869 0.5 1>}.RGB
   B=VectorParameter() {parameterName=Ray Color Bright, defaultValue=<rgba 1.6381 2.5554 20 1>}.RGB
   Alpha=@n6
  )
  n7 = Multiply(A=@n6, B=@n3.A)
  n8 = MaterialFunctionCall(
   Brightness=ScalarParameter() {defaultValue=1.5, parameterName=Fresnel Brightness}
   Exponent=ScalarParameter() {defaultValue=4, parameterName=Fresnel Exponent}
  ) {materialFunction=&MF_Fresnel_FX}
  LinearInterpolate(A=Multiply(A=Multiply(A=@n7, B=@n8.Fresnel), B=ScalarParameter() {defaultValue=0.5, parameterName=Opacity}), Alpha=@n5)
 ]
}
