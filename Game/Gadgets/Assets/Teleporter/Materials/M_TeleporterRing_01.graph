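# Translucent unlit material for teleporter ring mesh particles, usable in Niagara.
# An inverted Fresnel, optionally posterized into steps and replaced by a constant at low quality, blends inner and outer color and opacity.
# Switches choose particle or parameter colors and optional vertex-color opacity masking; particle alpha scales opacity, and emissive passes through a shared adjustment function.
# A Ring Size dynamic parameter multiplies the vertex normal, forming an offset for resizing the ring.
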Material /Game/Gadgets/Assets/Teleporter/Materials/M_TeleporterRing_01
graph {
 properties {
  bUsedWithMeshParticles=true
  bUsedWithNiagaraMeshParticles=true
  blendMode=BLEND_Translucent
  pixelDepthOffsetMode=PDOM_Legacy
  shadingModel=MSM_Unlit
  usedShadingModels=Unlit
 }
 nodes [
  n1 = ParticleColor()
  n2 = ScalarParameter() {defaultValue=1.7, parameterName=Fresnel Brightness}
  n3 = MaterialFunctionCall(Brightness=@n2, Exponent=ScalarParameter() {defaultValue=2, parameterName=Fresnel Exponent}) {materialFunction=&MF_Fresnel_FX}
  n4 = QualitySwitch(Default=@n3.Fresnel Inverted, Low=@n2, High=@n3.Fresnel Inverted, Medium=@n3.Fresnel Inverted)
  n5 = ScalarParameter() {defaultValue=2, parameterName=Steps}
  n6 = Multiply(A=@n4, B=@n5)
  n7 = StaticSwitchParameter(True=Divide(A=Floor(None=@n6), B=@n5), False=@n4) {defaultValue=true, parameterName=Use Steps?}
  n8 = LinearInterpolate(A=Divide(A=@n1.RGB, B=ScalarParameter() {defaultValue=6, parameterName=ColorInnerOpacity}) {constB=6}, B=@n1.RGB, Alpha=@n7)
  n9 = ScalarParameter() {parameterName=Desaturation}
  n10 = VectorParameter() {parameterName=Color Inner, defaultValue=<rgba 0.0125 0.1869 0.5 1>}
  n11 = VectorParameter() {parameterName=Color Outer, defaultValue=<rgba 1.2286 1.9166 15 1>}
  MaterialFunctionCall(Emissive=StaticSwitchParameter(True=Desaturation(None=@n8, Fraction=@n9), False=LinearInterpolate(A=@n10.RGB, B=@n11.RGB, Alpha=@n7)) {parameterName=UseParticleColor}) {materialFunction=&MF_FX_EmissiveAdjustments}
  Multiply(A=DynamicParameter() {defaultValue=<rgba 0 0 0 1>, paramNames [Ring Size, Param2, Param3, Param4]}.Ring Size, B=VertexNormalWS())
  n12 = Multiply(
   A=LinearInterpolate(
    A=ScalarParameter() {defaultValue=0.3, parameterName=Opacity Inner}
    B=ScalarParameter() {defaultValue=1, parameterName=Opacity Outer}
    Alpha=@n7
   ) {constA=0.3}
   B=@n1.A
  )
  n13 = VertexColor()
  StaticSwitchParameter(
   True=Multiply(A=@n12, B=StaticComponentMaskParameter(None=@n13) {defaultB=true, defaultG=true, defaultR=true, parameterName=VertexColorMask})
   False=@n12
  ) {parameterName=OpacityUseVertexColor?}
 ]
}
