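# Unlit translucent material for a teleporter ring, giving it a stylized holographic glow with animated marching bands.
# A stepped Fresnel blends inner and outer colors and opacities, while vertex green switches to the march color and time-driven, quantized opacity pulses along U.
# Vertex blue masks opacity, which is further modulated by scrolling screen-space scanlines.
# A Ring Size dynamic parameter scales the vertex normal, apparently to inflate the mesh.
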Material /Game/Gadgets/Assets/Teleporter/Materials/M_TeleporterRing_02
graph {
 properties {blendMode=BLEND_Translucent, pixelDepthOffsetMode=PDOM_Legacy, shadingModel=MSM_Unlit, usedShadingModels=Unlit}
 nodes [
  n1 = Panner(Coordinate=ScreenPosition().ViewportUV) {speedY=0.04}
  n2 = ScalarParameter() {defaultValue=135, parameterName="## of scanlines"}
  n3 = Frac(None=ComponentMask(None=Multiply(A=@n1, B=@n2) {constB=20}) {g=true})
  n4 = OneMinus(None=@n3)
  n5 = Ceil(None=Add(A=Multiply(A=@n3, B=@n4), B=ScalarParameter() {defaultValue=-0.15, parameterName=ScanlineSize}))
  Add(A=@n5)
  n6 = MaterialFunctionCall(
   Brightness=ScalarParameter() {defaultValue=1.7, parameterName=Fresnel Brightness}
   Exponent=ScalarParameter() {defaultValue=2, parameterName=Fresnel Exponent}
  ) {materialFunction=&MF_Fresnel_FX}
  n7 = ScalarParameter() {defaultValue=2, parameterName=Steps}
  n8 = Divide(A=Floor(None=Multiply(A=@n6.Fresnel Inverted, B=@n7)), B=@n7)
  LinearInterpolate(
   A=LinearInterpolate(
    A=VectorParameter() {parameterName=Color Inner, defaultValue=<rgba 0.0125 0.1869 0.5 1>}.RGB
    B=VectorParameter() {parameterName=Color Outer, defaultValue=<rgba 0.4095 0.6389 5 1>}.RGB
    Alpha=@n8
   )
   B=VectorParameter() {parameterName=Color March, defaultValue=<rgba 0.819 1.2777 10 1>}.RGB
   Alpha=VertexColor().G
  )
  Multiply(A=DynamicParameter() {defaultValue=<rgba 0 0 0 1>, paramNames [Ring Size, Param2, Param3, Param4]}.Ring Size, B=VertexNormalWS())
  n9 = LinearInterpolate(
   A=ScalarParameter() {defaultValue=0.15, parameterName=Opacity Inner}
   B=ScalarParameter() {defaultValue=0.5, parameterName=Opacity Outer}
   Alpha=@n8
  ) {constA=0.3}
  n10 = ComponentMask(None=TextureCoordinate()) {r=true}
  n11 = ScalarParameter() {defaultValue=4, parameterName=Steps A}
  n12 = Divide(A=Ceil(None=Multiply(A=@n10, B=@n11)), B=@n11)
  n13 = Time() {period=1}
  n14 = ScalarParameter() {defaultValue=1.3, parameterName=Time Scale}
  n15 = Multiply(A=Frac(None=Multiply(A=@n13, B=@n14))) {constB=2}
  n16 = Constant() {r=1}
  n17 = If(A=@n12, B=@n15, "A > B"=Constant(), "A == B"=@n16, "A < B"=@n16)
  n18 = ScalarParameter() {defaultValue=8, parameterName=Steps B}
  n19 = Divide(A=Ceil(None=Multiply(A=@n15, B=@n18)), B=@n18)
  n20 = Clamp(None=Add(A=OneMinus(None=@n19), B=@n12))
  n21 = Multiply(
   A=@n9
   B=LinearInterpolate(
    A=ScalarParameter() {defaultValue=0.025, parameterName=Opacity Marching A}
    B=ScalarParameter() {defaultValue=1, parameterName=Opacity Marching B}
    Alpha=Multiply(A=@n17, B=@n20)
   )
  )
  n22 = VertexColor()
  Multiply(
   A=Multiply(A=LinearInterpolate(A=@n9, B=@n21, Alpha=@n22.G), B=@n22.B)
   B=LinearInterpolate(
    A=ScalarParameter() {defaultValue=0.3, parameterName=Scanlines Dark}
    B=ScalarParameter() {defaultValue=1, parameterName=Scanlines Light}
    Alpha=Clamp(None=@n5)
   )
  )
 ]
}
