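# Translucent, unlit, two-sided mesh-particle material for the teleporter ring that fakes motion blur through a Blur dynamic parameter.
# Vertex colors and UVs shape a soft ring mask, which Blur breaks up with panning smoke noise offset by Random.
# That mask blends edge tint into bright particle color and drives opacity, with a view-angle fade off mobile.
# Blur also scales a local-Z stretch offset, weighted by vertex blue and UV.
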Material /Game/Gadgets/Assets/Teleporter/Materials/M_TeleporterRing_MotionBlur
graph {
 properties {
  bUsedWithMeshParticles=true
  blendMode=BLEND_Translucent
  pixelDepthOffsetMode=PDOM_Legacy
  shadingModel=MSM_Unlit
  twoSided=true
  usedShadingModels=Unlit
 }
 nodes [
  MaterialFunctionCall() {materialFunction=&DebugTimeSine}
  n1 = VectorParameter() {parameterName=EdgeTint, defaultValue=<rgba 0.1706 0.1506 0.5 1>}
  n2 = ParticleColor()
  n3 = TextureCoordinate()
  n4 = ComponentMask(None=@n3) {g=true}
  n5 = VertexColor()
  n6 = Constant() {r=0.5}
  n7 = ScalarParameter() {defaultValue=0.97, parameterName=OuterRingHardness}
  n8 = Multiply(
   A=SphereMask(A=LinearInterpolate(A=@n4, B=@n3, Alpha=@n5.G), B=@n6, Radius=@n6, Hardness=Multiply(A=@n7, B=@n5.G) {constB=0.95}) {hardnessPercent=0}
   B=@n5.R
  )
  n9 = OneMinus(None=@n5.G)
  n10 = DynamicParameter() {defaultValue=<rgba 1 1 1 0.5>, paramNames [Param1,Param2,Random,Blur]}
  n11 = Panner(Coordinate=Add(A=TextureCoordinate() {uTiling=4, vTiling=0.02}, B=@n10.Random)) {speedX=0.1, speedY=-1}
  n12 = Multiply(
   A=Add(A=TextureSample(UVs=@n11) {texture=&T_Inky_Smoke_Tile_Chromatic}.RGB) {constB=0.25}
   B=ScalarParameter() {defaultValue=0.75, parameterName=BlurBrightness}
  )
  n13 = Clamp(None=@n10.Blur)
  n14 = Multiply(A=@n8, B=LinearInterpolate(B=Multiply(A=@n9, B=@n12), Alpha=Power(Base=@n13) {constExponent=0.6}) {constA=1})
  MaterialFunctionCall(Emissive=LinearInterpolate(A=Multiply(A=@n1.RGB, B=@n2.RGB), B=Multiply(A=@n2.RGB) {constA=3, constB=5}, Alpha=@n14)) {materialFunction=&MF_FX_EmissiveAdjustments}
  n15 = DotProduct(A=CameraVectorWS(), B=VertexNormalWS())
  n16 = Multiply(A=Clamp(None=Abs(None=@n15)), B=@n8)
  Multiply(A=@n2.A, B=Multiply(A=ShadingPathSwitch(Default=@n16, Mobile=@n8) {desc=Migrated from FeatureLevelSwitch}, B=@n14))
  n17 = TextureCoordinate()
  n18 = Multiply(
   A=LinearInterpolate(Alpha=ComponentMask(None=@n17) {g=true}) {constA=1.25, constB=-0.25}
   B=Transform(None=Constant3Vector() {constant=<rgba 0 0 1 0>}) {transformSourceType=TRANSFORMSOURCE_Local}
  )
  Multiply(A=Multiply(A=Multiply(A=@n5.B, B=@n18), B=ScalarParameter() {defaultValue=100, parameterName=MaxStretch}) {constA=100}, B=@n10.Blur)
 ]
}
