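# Unlit masked material for a teleporter gadget's vortex effect, with a stylized, banded look over scrolling noise.
# Two panned noise channels are summed and quantized with Ceil, driving an inner-to-outer color blend with inverted vertex alpha and feeding an Erode offset.
# Summed scrolling ripple normals are also stepped, then bend a reflection vector that samples a texture for tinted highlights.
# Unquantized noise times a WPO vector is scaled by vertex alpha.
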Material /Game/Gadgets/Assets/Teleporter/Materials/M_TeleporterVortex_01
graph {
 properties {blendMode=BLEND_Masked, pixelDepthOffsetMode=PDOM_Legacy, shadingModel=MSM_Unlit, usedShadingModels=Unlit}
 nodes [
  n1 = TextureCoordinate() {uTiling=2, vTiling=1.6}
  n2 = TextureCoordinate() {uTiling=2, vTiling=4.6}
  n3 = Add(
   A=TextureSample(UVs=Panner(Coordinate=@n1) {bFractionalPart=true, speedX=0.01, speedY=-0.4}) {texture=&T_NoisePack_01}.R
   B=TextureSample(UVs=Panner(Coordinate=@n2) {bFractionalPart=true, speedX=-0.02, speedY=-0.88}) {texture=&T_NoisePack_01}.G
  )
  n4 = ScalarParameter() {defaultValue=2, parameterName=Steps}
  n5 = Multiply(A=@n3, B=@n4)
  n6 = Divide(A=Divide(A=Ceil(None=@n5), B=@n4))
  n7 = Add(A=@n6, B=OneMinus(None=VertexColor().A))
  n8 = ScalarParameter() {defaultValue=1.7, parameterName=LERP Expononet}
  n9 = Panner(Coordinate=TextureCoordinate() {vTiling=3.34}) {bFractionalPart=true, speedX=-0.0233, speedY=-0.18}
  n10 = Panner(Coordinate=TextureCoordinate() {vTiling=4.17}) {bFractionalPart=true, speedX=0.0367, speedY=-0.7}
  n11 = ScalarParameter() {defaultValue=2, parameterName=Steps}
  n12 = Multiply(
   A=Add(
    A=TextureSample(UVs=@n9) {samplerType=SAMPLERTYPE_Normal, texture=/Game/Effects/Fort_Effects/Gameplay/Water/Textures/T_WaterRipples_01}.RGB
    B=TextureSample(UVs=@n10) {samplerType=SAMPLERTYPE_Normal, texture=&T_WaterRipples_01}.RGB
   )
   B=@n11
  )
  n13 = Normalize(VectorInput=Divide(A=Ceil(None=@n12), B=@n11))
  n14 = TextureSample(UVs=ComponentMask(None=ReflectionVectorWS(CustomWorldNormal=@n13)) {g=true, r=true}) {texture=/Game/Effects/Fort_Effects/Gameplay/Water/Textures/FX_Reflection_03_Diff}
  n15 = ScalarParameter() {defaultValue=1.7, parameterName=Reflection Vector Contrast}
  Add(
   A=LinearInterpolate(
    A=VectorParameter() {parameterName=Color Inner, defaultValue=<rgba 0 0.0993 0.535 1>}.RGB
    B=VectorParameter() {parameterName=Color Outer, defaultValue=<rgba 0.819 1.2777 10 1>}.RGB
    Alpha=Power(Base=@n7, Exp=@n8) {constExponent=3}
   )
   B=Multiply(A=Power(Base=@n14.RGB, Exp=@n15), B=VectorParameter() {parameterName=Reflection Vector Color, defaultValue=<rgba 1.8504 4.9013 170 0>}.RGB)
  ) {constB=0}
  Add(A=@n6, B=ScalarParameter() {defaultValue=1, parameterName=Erode})
  Multiply(A=Multiply(A=@n3, B=VectorParameter() {parameterName=WPO, defaultValue=<rgba 0 0 3 0>}.RGB), B=VertexColor().A)
  Sine(None=Time())
 ]
}
