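# Unlit, two-sided translucent teleporter glow material, usable on skeletal meshes and named as the variant without refraction.
# Two counter-rotating, panning sparkle noises combine with inverted fresnel and optional depth-intersection glow to blend inner and outer colors.
# A Transition parameter, with a debug sine switch, drives a grid-square dissolve along a vertical UV gradient with bright fresnel-weighted edges.
# Opacity combines noise and fresnel, masked by the gradient, grid dissolve and near-camera fade.
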Material /Game/Gadgets/Assets/Teleporter/Materials/M_Teleporter_Glow_NoRefraction
graph {
 properties {
  bUsedWithMeshDeformer=true
  bUsedWithSkeletalMesh=true
  blendMode=BLEND_Translucent
  pixelDepthOffsetMode=PDOM_Legacy
  shadingModel=MSM_Unlit
  twoSided=true
  usedShadingModels=Unlit
 }
 nodes [
  n1 = ScalarParameter() {defaultValue=16, parameterName=Glitter A Repeat U}
  n2 = ScalarParameter() {defaultValue=6, parameterName=Glitter A Repeat V}
  n3 = Rotator(Coordinate=Multiply(A=TextureCoordinate(), B=AppendVector(A=@n1, B=@n2))) {speed=0.05}
  n4 = ScalarParameter() {defaultValue=11.2, parameterName=Glitter B Repeat U}
  n5 = ScalarParameter() {defaultValue=5, parameterName=Glitter B Repeat V}
  n6 = Rotator(Coordinate=Multiply(A=TextureCoordinate(), B=AppendVector(A=@n4, B=@n5))) {speed=-0.1056}
  n7 = Multiply(
   A=TextureSample(UVs=Panner(Coordinate=@n3) {speedX=0.21, speedY=0.147}) {samplerType=SAMPLERTYPE_LinearColor, texture=/Game/Gadgets/Assets/Slow_Field/Textures/T_Slowfield_SparkleNoise_01}.R
   B=TextureSample(UVs=Panner(Coordinate=@n6) {speedX=-0.154, speedY=-0.077}) {samplerType=SAMPLERTYPE_LinearColor, texture=&T_Slowfield_SparkleNoise_01}.G
  )
  n8 = Divide(A=Subtract(A=SceneDepth().R, B=PixelDepth().R), B=ScalarParameter() {defaultValue=8, parameterName=Intersect Scale})
  n9 = Multiply(A=OneMinus(None=Clamp(None=@n8)), B=ScalarParameter() {defaultValue=1, parameterName=Intersect Boost})
  n10 = Constant()
  n11 = MaterialFunctionCall(
   Brightness=ScalarParameter() {defaultValue=1.04, parameterName=Fresnel Brightness}
   Exponent=ScalarParameter() {defaultValue=3, parameterName=Fresnel Exponent}
  ) {materialFunction=&MF_Fresnel_FX}
  n12 = Add(A=@n7, B=Add(A=StaticSwitchParameter(True=@n9, False=@n10) {defaultValue=true, parameterName=Depth Intersect?}, B=@n11.Fresnel Inverted))
  n13 = ScalarParameter() {defaultValue=2, parameterName=Grid U Repeat}
  n14 = ScalarParameter() {defaultValue=1.5, parameterName=Grid V Repeat}
  n15 = TextureSample(UVs=Multiply(A=TextureCoordinate(), B=AppendVector(A=@n13, B=@n14))) {samplerType=SAMPLERTYPE_LinearColor, texture=/Game/Gadgets/Assets/Teleporter/Textures/T_Teleporter_GridSquares_01}
  n16 = StaticSwitchParameter(True=Sine(None=Time()) {period=2}, False=ScalarParameter() {defaultValue=1, parameterName=Transition}) {parameterName=DebugTransition?}
  n17 = TextureCoordinate()
  n18 = Power(
   Base=Multiply(A=ComponentMask(None=@n17) {g=true}, B=ScalarParameter() {defaultValue=1.5, parameterName=Gradient Opacity Brightness})
   Exp=ScalarParameter() {defaultValue=3, parameterName=Gradient Opacity Contrast}
  ) {constExponent=0.8}
  n19 = Add(A=@n18, B=@n16)
  n20 = Clamp(None=Multiply(A=Add(A=@n15.R, B=@n16), B=Clamp(None=@n19)))
  n21 = Multiply(A=Multiply(A=@n20, B=OneMinus(None=@n16)), B=ScalarParameter() {defaultValue=25, parameterName=Transition Edge Brightness}) {constB=5}
  LinearInterpolate(
   A=VectorParameter() {parameterName=Color Inner, defaultValue=<rgba 0.0125 0.1869 0.5 1>}.RGB
   B=VectorParameter() {parameterName=Color Outer, defaultValue=<rgba 1.2286 1.9166 15 1>}.RGB
   Alpha=Add(A=Clamp(None=@n12), B=Multiply(A=@n11.Fresnel, B=@n21))
  )
  n22 = Add(
   A=@n7
   B=LinearInterpolate(A=ScalarParameter() {defaultValue=0.15, parameterName=Opacity Core}, Alpha=@n11.Fresnel Inverted) {constA=0.1, constB=0.85}
  )
  n23 = MaterialFunctionCall(Distance=ScalarParameter() {defaultValue=100, parameterName=Near Fade}) {materialFunction=&Near_Camera_Fade}
  Multiply(A=Multiply(A=Multiply(A=@n22, B=@n23.Result), B=Clamp(None=@n18)), B=@n20)
 ]
}
