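# Translucent unlit mesh-particle material for teleporter loading rings that fill in stepped segments to show charge progress.
# Dynamic parameter Progress switches segments from uncharged to charged colors and opacity, while uncharged segments run a looping stepped sweep offset by TimeSpeed.
# Fresnel and scrolling scanlines shape the glow and opacity, which also passes through a distance fade.
# Panning noise drives world-position wobble, with intermittent glitch bursts from two multiplied sine waves.
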Material /Game/Gadgets/Assets/Teleporter/Materials/M_Teleporter_LoadingRings_01
graph {
 properties {bUsedWithMeshParticles=true, blendMode=BLEND_Translucent, pixelDepthOffsetMode=PDOM_Legacy, shadingModel=MSM_Unlit, usedShadingModels=Unlit}
 nodes [
  n1 = Panner(Coordinate=TextureCoordinate() {coordinateIndex=1}) {speedX=1}
  n2 = TextureSample(UVs=Panner(Coordinate=TextureCoordinate() {coordinateIndex=1, uTiling=0.05, vTiling=10}) {speedY=2}) {samplerType=SAMPLERTYPE_Normal, texture=/Game/Global/Textures/Default/Noise/64x64Noise}
  n3 = VectorParameter() {parameterName=WPO Intensity B, defaultValue=<rgba 0 5 0 0>}
  n4 = Time()
  n5 = Sine(None=@n4) {period=0.9761}
  n6 = Sine(None=@n4) {period=1.915}
  n7 = Clamp(None=Multiply(A=ConstantBiasScale(None=@n5), B=ConstantBiasScale(None=@n6)))
  Add(
   A=Multiply(
    A=TextureSample(UVs=@n1) {samplerType=SAMPLERTYPE_Normal, texture=&64x64Noise}.RGB
    B=VectorParameter() {parameterName=WPO Intensity A, defaultValue=<rgba 0.4 0.4 1.2 0>}.RGB
   )
   B=Multiply(A=Multiply(A=@n2.RGB, B=@n3.RGB), B=Power(Base=@n7) {constExponent=4})
  )
  n8 = VectorParameter() {parameterName=Color Inner Uncharged, defaultValue=<rgba 1 0.2383 0.133 1>}
  n9 = VectorParameter() {parameterName=Color Outer Uncharged, defaultValue=<rgba 4 2.7247 0.4143 1>}
  n10 = ComponentMask(None=TextureCoordinate()) {r=true}
  n11 = ScalarParameter() {defaultValue=7, parameterName=Steps}
  n12 = Divide(A=Ceil(None=Multiply(A=@n10, B=@n11)), B=@n11)
  n13 = Time() {period=1}
  n14 = DynamicParameter() {defaultValue=<rgba 0.3 1 1 1>, paramNames [Progress,TimeSpeed,Param3,Param4]}
  n15 = Multiply(A=Frac(None=Add(A=@n13, B=@n14.TimeSpeed))) {constB=2}
  n16 = Constant() {r=1}
  n17 = ScalarParameter() {defaultValue=7, parameterName=Steps}
  n18 = Multiply(A=@n15, B=@n17)
  n19 = OneMinus(None=Divide(A=Ceil(None=@n18), B=@n17))
  n20 = Multiply(A=If(A=@n12, B=@n15, "A > B"=Constant(), "A == B"=@n16, "A < B"=@n16), B=Clamp(None=Add(A=@n19, B=@n12)))
  n21 = MaterialFunctionCall(
   Brightness=ScalarParameter() {defaultValue=1.7, parameterName=Fresnel Brightness}
   Exponent=ScalarParameter() {defaultValue=4, parameterName=Fresnel Exponent}
  ) {materialFunction=&MF_Fresnel_FX}
  n22 = Clamp(None=Add(A=@n20, B=@n21.Fresnel Inverted))
  n23 = VectorParameter() {parameterName=Color Inner Charged, defaultValue=<rgba 0.9829 1.5333 12 1>}
  n24 = VectorParameter() {parameterName=Color Outer Charged, defaultValue=<rgba 0.2568 0.7602 3 1>}
  n25 = ScalarParameter() {defaultValue=7, parameterName=Steps}
  n26 = Multiply(A=ComponentMask(None=TextureCoordinate()) {r=true}, B=@n25)
  n27 = Constant() {r=1}
  n28 = If(
   A=Divide(A=Ceil(None=@n26), B=@n25)
   B=DynamicParameter() {defaultValue=<rgba 0.3 1 1 1>, paramNames [Progress,TimeSpeed,Param3,Param4]}.Progress
   "A > B"=Constant()
   "A == B"=@n27
   "A < B"=@n27
  )
  MaterialFunctionCall(Emissive=LinearInterpolate(
    A=LinearInterpolate(A=@n8.RGB, B=@n9.RGB, Alpha=@n22) {constAlpha=0}
    B=LinearInterpolate(A=@n23.RGB, B=@n24.RGB, Alpha=@n21.Fresnel Inverted)
    Alpha=@n28
   )) {materialFunction=&MF_FX_EmissiveAdjustments}
  MaterialFunctionCall!(In) {materialFunction=/Engine/Functions/Engine_MaterialFunctions02/Utility/Round}
  n29 = ScalarParameter() {defaultValue=0.7, parameterName=Opacity Full}
  n30 = LinearInterpolate(
   A=LinearInterpolate(
    A=ScalarParameter() {defaultValue=0.05, parameterName=Opacity Charging A}
    B=ScalarParameter() {defaultValue=0.5, parameterName=Opacity Charging B}
    Alpha=@n20
   )
   B=@n29
   Alpha=Multiply(A=@n28, B=@n21.Fresnel Inverted)
  )
  n31 = ScalarParameter() {defaultValue=0.3, parameterName=Scanlines Dark}
  n32 = ScalarParameter() {defaultValue=1, parameterName=Scanlines Light}
  n33 = Panner(Coordinate=TextureCoordinate() {coordinateIndex=1}) {speedY=0.1}
  n34 = Frac(None=ComponentMask(None=Multiply(A=@n33) {constB=20}) {g=true})
  n35 = Add(A=Multiply(A=@n34, B=OneMinus(None=@n34)), B=ScalarParameter() {defaultValue=-0.15, parameterName=ScanlineSize})
  n36 = Clamp(None=Add(A=Ceil(None=@n35), B=@n21.Fresnel Inverted))
  MaterialFunctionCall(
   Opacity In=Multiply(A=LinearInterpolate(A=@n30, B=@n29, Alpha=@n28), B=LinearInterpolate(A=@n31, B=@n32, Alpha=@n36))
   Fade_Start_Distance=ScalarParameter() {defaultValue=1500, parameterName=Start}
   Fade_End_Distance=ScalarParameter() {defaultValue=900, parameterName=End}
  ) {materialFunction=&MF_Distance_SyncedFade_FX}
  ScalarParameter() {defaultValue=1.1, parameterName=Refract}
 ]
}
