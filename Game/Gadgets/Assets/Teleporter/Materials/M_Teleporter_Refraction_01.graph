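# Unlit translucent teleporter VFX material for particles, meshes and ribbons that combines refraction distortion with colored plasma.
# Index-of-refraction distortion is driven by a dynamic particle parameter, masked by vertex color and panning noise, and faded by scene depth and camera distance.
# Two panning plasma samples, with per-particle tiling variation, blend inner and outer colors tinted by particle color.
# The IsRibbon? switch swaps UV axes and adds an edge gradient to opacity.
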Material /Game/Gadgets/Assets/Teleporter/Materials/M_Teleporter_Refraction_01
graph {
 properties {
  bUsedWithBeamTrails=true
  bUsedWithMeshParticles=true
  bUsedWithNiagaraMeshParticles=true
  bUsedWithParticleSprites=true
  blendMode=BLEND_Translucent
  pixelDepthOffsetMode=PDOM_Legacy
  refractionMethod=RM_IndexOfRefraction
  shadingModel=MSM_Unlit
  usedShadingModels=Unlit
 }
 nodes [
  n1 = DynamicParameter() {defaultValue=<rgba 1 0.05 0 1>, paramNames ["Refaction Anim (0-1)", Refaction Random, Param3, Param4]}
  n2 = VertexColor()
  n3 = LinearInterpolate(B=Add(A=Multiply(A=@n1."Refaction Anim (0-1)", B=@n1."Refaction Anim (0-1)")), Alpha=Multiply(A=@n2.R, B=@n2.R)) {constA=1}
  n4 = DepthFade(FadeDistance=ScalarParameter() {defaultValue=100, parameterName=Refraction Depth Fade Distance})
  n5 = TextureCoordinate() {coordinateIndex=1, vTiling=0.7}
  n6 = StaticSwitchParameter(True=AppendVector(A=ComponentMask(None=@n5) {g=true}, B=ComponentMask(None=@n5) {r=true}), False=TextureCoordinate() {vTiling=0.7}) {parameterName=IsRibbon?}
  n7 = Power(
   Base=TextureSample(UVs=Panner(Coordinate=@n6) {bFractionalPart=true, speedY=0.184}) {texture=&T_NoisePack_01}.R
   Exp=ScalarParameter() {defaultValue=0.5, parameterName=Cloud Power}
  )
  n8 = PixelDepth()
  n9 = ScalarParameter() {defaultValue=150, parameterName=Refraction Camera Fade Distance Scale}
  LinearInterpolate(B=LinearInterpolate(B=LinearInterpolate(B=@n3, Alpha=@n4) {constA=1}, Alpha=@n7) {constA=1}, Alpha=Clamp(None=Divide(A=@n8.R, B=@n9))) {constA=1}
  n10 = VectorParameter() {parameterName=Color Outer, defaultValue=<rgba 0 0.1634 0.87 1>}
  n11 = VectorParameter() {parameterName=Color Inner, defaultValue=<rgba 0.1 0.8164 2 0>}
  n12 = TextureCoordinate() {coordinateIndex=1}
  n13 = DynamicParameter() {defaultValue=<rgba 1 0.05 0 1>, paramNames ["Refaction Anim (0-1)", Refaction Random, Param3, Param4]}
  n14 = StaticSwitchParameter(
   True=AppendVector(A=ComponentMask(None=@n12) {g=true}, B=ComponentMask(None=@n12) {r=true})
   False=LinearInterpolate(
    A=TextureCoordinate()
    B=TextureCoordinate() {coordinateIndex=1}
    Alpha=Multiply(A=@n13."Refaction Anim (0-1)", B=@n13."Refaction Anim (0-1)")
   )
  ) {parameterName=IsRibbon?}
  n15 = ScalarParameter() {defaultValue=2, parameterName=U Repeat A}
  n16 = Multiply(A=@n13.Refaction Random) {constB=10}
  n17 = Multiply(A=ScalarParameter() {defaultValue=0.088, parameterName=V Repeat A}, B=@n16)
  n18 = AppendVector(A=@n13.Param4, B=Constant())
  n19 = Multiply(A=@n14, B=Add(A=AppendVector(A=@n15, B=@n17), B=@n18))
  n20 = ScalarParameter() {defaultValue=1, parameterName=U Repeat B}
  n21 = Multiply(A=@n16, B=ScalarParameter() {defaultValue=0.14, parameterName=V Repeat B})
  n22 = Multiply(A=@n14, B=Add(A=@n18, B=AppendVector(A=@n20, B=@n21)))
  n23 = Add(
   A=TextureSample(UVs=Panner(Coordinate=@n19) {bFractionalPart=true, speedX=0.031, speedY=0.108}) {texture=/Game/Effects/Fort_Effects/Textures/Plasma/T_PlasmaNoise_A_01}.B
   B=TextureSample(UVs=Panner(Coordinate=@n22) {bFractionalPart=true, speedX=-0.099, speedY=0.131}) {texture=&T_PlasmaNoise_A_01}.R
  )
  n24 = VertexColor()
  n25 = Power(Base=Multiply(A=@n24.R, B=@n24.R), Exp=ScalarParameter() {defaultValue=2, parameterName=Color LERP Power})
  n26 = Multiply(A=Multiply(A=Divide(A=@n23), B=@n7), B=@n25)
  n27 = ParticleColor()
  MaterialFunctionCall(Emissive=Multiply(A=LinearInterpolate(A=@n10.RGB, B=@n11.RGB, Alpha=@n26), B=@n27.RGB)) {materialFunction=&MF_FX_EmissiveAdjustments}
  Multiply(
   A=LinearInterpolate(
    A=ScalarParameter() {defaultValue=1, parameterName=Opacity Outer}
    B=ScalarParameter() {defaultValue=1, parameterName=Opacity Inner}
    Alpha=@n26
   )
   B=@n25
  )
  n28 = Multiply(A=@n27.A, B=Multiply(A=Multiply(A=@n25, B=@n7), B=@n4))
  n29 = ComponentMask(None=TextureCoordinate()) {g=true}
  n30 = Multiply(A=Multiply(A=@n29, B=OneMinus(None=@n29)), B=ScalarParameter() {defaultValue=2, parameterName=Gradient Brightness})
  n31 = ScalarParameter() {defaultValue=2, parameterName=Gradient Contrast}
  StaticSwitchParameter(True=Multiply(A=@n28, B=Power(Base=@n30, Exp=@n31)), False=@n28) {parameterName=IsRibbon?}
 ]
}
