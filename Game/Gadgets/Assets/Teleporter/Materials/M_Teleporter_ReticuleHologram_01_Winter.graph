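# Translucent unlit material, Nanite-enabled, for the winter variant of the teleporter's reticule hologram, giving a two-tone blue look with a distinct outline.
# An inverted Fresnel from MF_Fresnel_FX plus a bias passes through Ceil, turning the soft falloff into a hard-stepped inner/outline mask.
# That shared mask blends inner and outline colors, and inner and outer opacities, all exposed as parameters along with the Fresnel controls.
# Responsive AA is enabled.
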Material /Game/Gadgets/Assets/Teleporter/Materials/M_Teleporter_ReticuleHologram_01_Winter
graph {
 properties {
  bEnableResponsiveAA=true
  bUsedWithNanite=true
  blendMode=BLEND_Translucent
  pixelDepthOffsetMode=PDOM_Legacy
  shadingModel=MSM_Unlit
  usedShadingModels=Unlit
 }
 nodes [
  n1 = ScalarParameter() {defaultValue=1.3, parameterName=Fresnel Brightness}
  n2 = ScalarParameter() {defaultValue=10, parameterName=Fresnel Exponent}
  n3 = Ceil(None=Add(
    A=MaterialFunctionCall(Brightness=@n1, Exponent=@n2) {materialFunction=&MF_Fresnel_FX}.Fresnel Inverted
    B=ScalarParameter() {defaultValue=-0.4, parameterName=Frensel Bias}
   ) {constB=-0.5})
  LinearInterpolate(
   A=ScalarParameter() {defaultValue=0.9, parameterName=Inner Opacity}
   B=ScalarParameter() {defaultValue=0.75, parameterName=Outer Opacity}
   Alpha=@n3
  ) {constA=0.1}
  LinearInterpolate(
   A=VectorParameter() {parameterName=Color Inner, defaultValue=<rgba 0 0.0507 0.26 0>}.RGB
   B=VectorParameter() {parameterName=Color Outline, defaultValue=<rgba 0.09 0.2674 1 0>}.RGB
   Alpha=@n3
  ) {constA=0.1}
 ]
}
